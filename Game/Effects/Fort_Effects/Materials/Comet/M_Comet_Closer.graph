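# M_Comet_Closer is an unlit, additive, two-sided mesh-particle material for a comet effect, apparently its close-up version.
# Two panning noise textures, shaped along V, blend dark and light colors, topped with a UV1 center glow and squared dust debris.
# Time-of-day functions tint the HDR-remapped emissive and scale opacity.
# Particle alpha, noise and a view-angle fade, adjusted along the particle's local X axis, soften the mesh edges.
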Material /Game/Effects/Fort_Effects/Materials/Comet/M_Comet_Closer
graph {
 properties {
  bUsedWithMeshParticles=true
  bUsedWithNiagaraMeshParticles=true
  blendMode=BLEND_Additive
  pixelDepthOffsetMode=PDOM_Legacy
  shadingModel=MSM_Unlit
  twoSided=true
  usedShadingModels=Unlit
 }
 nodes [
  n1 = VectorParameter() {parameterName=EveningTint, defaultValue=<rgba 1 0.647 0.577 0>}
  n2 = VectorParameter() {parameterName=DayTint, defaultValue=<rgba 0.4479 1 0.8277 0>}
  n3 = VectorParameter() {parameterName=NightTint, defaultValue=<rgba 0.3398 0.1245 0.75 0>}
  n4 = Power(Base=SphereMask(A=TextureCoordinate() {coordinateIndex=1}, B=Constant() {r=0.5}) {attenuationRadius=0.25, hardnessPercent=0}) {constExponent=5}
  n5 = VectorParameter() {parameterName=CenterColor, defaultValue=<rgba 0.4531 0.7178 1 0>}
  n6 = Power(Base=ComponentMask(None=TextureCoordinate()) {g=true})
  n7 = ScalarParameter() {defaultValue=1, parameterName=Texture UV Scale}
  n8 = Multiply(A=TextureCoordinate() {uTiling=4, vTiling=0.7}, B=@n7)
  n9 = Multiply(A=Time(), B=ScalarParameter() {defaultValue=0.6, parameterName=PanSpeed})
  n10 = Multiply(A=TextureCoordinate() {uTiling=3, vTiling=0.6}, B=@n7)
  n11 = LinearInterpolate(
   A=TextureSample(UVs=Panner(Coordinate=@n8, Time=@n9) {speedX=0.04, speedY=0.07}) {samplerType=SAMPLERTYPE_LinearGrayscale, texture=/Game/Global/Textures/Default/Noise/T_SwirlNoise_N_height}.R
   B=TextureSample(UVs=Panner(Coordinate=@n10, Time=@n9) {speedX=-0.05, speedY=0.16}) {texture=&T_Smooth_tile}.R
  )
  n12 = OneMinus(None=ComponentMask(None=TextureCoordinate()) {g=true})
  n13 = Multiply(A=@n6, B=Power(Base=Multiply(A=@n11) {constB=1.5}, Exp=LinearInterpolate(Alpha=@n12) {constA=0.6, constB=4}))
  n14 = LinearInterpolate(
   A=VectorParameter() {parameterName=DarkColor, defaultValue=<rgba 0.1292 0.1853 0.4 0>}.RGB
   B=VectorParameter() {parameterName=LightColor, defaultValue=<rgba 0.0334 0.548 1.5 0>}.RGB
   Alpha=@n13
  )
  n15 = ScalarParameter() {defaultValue=1, parameterName=Texture UV Scale}
  n16 = Multiply(A=TextureCoordinate() {uTiling=3}, B=@n15)
  n17 = Multiply(A=TextureCoordinate() {uTiling=4, vTiling=2}, B=@n15)
  n18 = Add(
   A=TextureSample(UVs=Panner(Coordinate=@n16, Time=@n9) {speedX=-0.01, speedY=0.025}) {texture=/Game/Global/Textures/Default/PointSplatter/T_Tiling_Dust}.RGB
   B=TextureSample(UVs=Panner(Coordinate=@n17, Time=@n9) {speedX=0.01, speedY=0.05}) {texture=&T_Tiling_Dust}.RGB
  )
  n19 = Add(
   A=Add(A=Multiply(A=@n4, B=@n5.RGB), B=Clamp(None=@n14) {maxDefault=10})
   B=Multiply(A=Multiply(A=@n18, B=@n18), B=VectorParameter() {parameterName=DebrisColor, defaultValue=<rgba 3.2956 3.6779 7 0>}.RGB)
  )
  n20 = Clamp(None=@n19)
  n21 = Multiply(A=@n19, B=ScalarParameter() {defaultValue=2, parameterName=HDRMult})
  MaterialFunctionCall(Emissive=Multiply(
    A=MaterialFunctionCall(Color_Morning=@n1.RGB, Color_Day=@n2.RGB, Color_Evening=@n1.RGB, Color_Night=@n3.RGB) {materialFunction=&Color_Time_of_Day}.Result
    B=MaterialFunctionCall(LDR Input=@n20, HDR Tint=@n21) {materialFunction=&DeriveHDRfromLDR}.RemappedRange
   )) {materialFunction=&MF_FX_EmissiveAdjustments}
  n22 = ParticleColor()
  n23 = Add(A=@n6, B=@n13)
  n24 = Constant() {r=0.4}
  n25 = Constant() {r=0.12}
  n26 = Constant() {r=1}
  n27 = CameraVectorWS()
  n28 = DotProduct(A=@n27, B=VertexNormalWS())
  n29 = Normalize(VectorInput=Transform(None=Constant3Vector() {constant=<rgba 1 0 0 0>}) {transformSourceType=TRANSFORMSOURCE_Local, transformType=TRANSFORM_ParticleWorld})
  n30 = Abs(None=DotProduct(A=@n27, B=Abs(None=@n29)))
  n31 = ScalarParameter() {defaultValue=6, parameterName=DirectionalFadePower}
  n32 = Power(
   Base=LinearInterpolate(A=Abs(None=@n28), Alpha=Power(Base=@n30, Exp=@n31) {constExponent=4}) {constB=1.25}
   Exp=ScalarParameter() {defaultValue=5, parameterName=AngleFadePower}
  ) {constExponent=4}
  Multiply(
   A=Multiply(
    A=Multiply(A=@n22.A, B=@n23)
    B=MaterialFunctionCall(Color_Morning=@n24, Color_Day=@n25, Color_Evening=@n24, Color_Night=@n26) {materialFunction=&Color_Time_of_Day}.Result
   )
   B=Clamp(None=Multiply(A=@n32) {constB=4})
  )
 ]
}
