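# Additive, unlit, two-sided mesh particle material for the comet rock effect.
# Two panning noise textures blend dark and light colors, with a V-based highlight at one end and fade to a rear color at the other.
# The color gets HDR remapping and a time-of-day tint, then passes through emissive adjustments.
# A second chain combines particle alpha, time-of-day strength (strongest at night) and a view-angle fade.
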Material /Game/Effects/Fort_Effects/Materials/Comet/M_Comet_Rock
graph {
 properties {
  bUsedWithMeshParticles=true
  bUsedWithNiagaraMeshParticles=true
  blendMode=BLEND_Additive
  pixelDepthOffsetMode=PDOM_Legacy
  shadingModel=MSM_Unlit
  twoSided=true
  usedShadingModels=Unlit
 }
 nodes [
  n1 = VectorParameter() {parameterName=EveningTint, defaultValue=<rgba 0.8 0.5555 0.4375 0>}
  n2 = VectorParameter() {parameterName=DayTint, defaultValue=<rgba 0.175 0.7 0.5361 0>}
  n3 = VectorParameter() {parameterName=NightTint, defaultValue=<rgba 0.4531 0.166 1 0>}
  n4 = TextureCoordinate()
  n5 = Power(Base=OneMinus(None=ComponentMask(None=@n4) {g=true})) {constExponent=8}
  n6 = VectorParameter() {parameterName=LightColor, defaultValue=<rgba 0.2292 0.8506 2 0>}
  n7 = VectorParameter() {parameterName=DarkColor, defaultValue=<rgba 0.2101 0.2615 0.4583 0>}
  n8 = TextureCoordinate() {uTiling=2}
  n9 = Multiply(A=Time(), B=ScalarParameter() {defaultValue=0.3, parameterName=PanSpeed})
  n10 = TextureCoordinate() {uTiling=2}
  n11 = LinearInterpolate(
   A=TextureSample(UVs=Panner(Coordinate=@n8, Time=@n9) {speedX=0.04, speedY=-0.08}) {samplerType=SAMPLERTYPE_LinearGrayscale, texture=/Game/Global/Textures/Default/Noise/T_SwirlNoise_N_height}.R
   B=TextureSample(UVs=Panner(Coordinate=@n10, Time=@n9) {speedX=-0.05, speedY=-0.15}) {texture=&T_Smooth_tile}.R
  ) {constAlpha=0.4}
  n12 = ComponentMask(None=TextureCoordinate()) {g=true}
  n13 = Multiply(A=@n12, B=@n12)
  n14 = LinearInterpolate(
   A=Add(A=Multiply(A=@n5, B=@n6.RGB), B=LinearInterpolate(A=@n7.RGB, B=@n6.RGB, Alpha=@n11))
   B=VectorParameter() {parameterName=RearColor, defaultValue=<rgba 0.1572 0.1328 0.1771 0>}.RGB
   Alpha=Multiply(A=@n13, B=@n13)
  )
  n15 = Clamp(None=@n14)
  n16 = Multiply(A=@n14, B=ScalarParameter() {defaultValue=6, parameterName=HDRMult})
  MaterialFunctionCall(Emissive=Multiply(
    A=MaterialFunctionCall(Color_Morning=@n1.RGB, Color_Day=@n2.RGB, Color_Evening=@n1.RGB, Color_Night=@n3.RGB) {materialFunction=&Color_Time_of_Day}.Result
    B=MaterialFunctionCall(LDR Input=@n15, HDR Tint=@n16) {materialFunction=&DeriveHDRfromLDR}.RemappedRange
   )) {materialFunction=&MF_FX_EmissiveAdjustments}
  n17 = Constant() {r=0.4}
  n18 = Constant() {r=0.1}
  n19 = Constant() {r=1}
  n20 = MaterialFunctionCall("FadeSensitivity (optional)"=ScalarParameter() {defaultValue=2.5, parameterName=FadeAngle}) {materialFunction=&FadeBasedOnViewAngle}
  Multiply(
   A=Multiply(
    A=ParticleColor().A
    B=MaterialFunctionCall(Color_Morning=@n17, Color_Day=@n18, Color_Evening=@n17, Color_Night=@n19) {materialFunction=&Color_Time_of_Day}.Result
   )
   B=Clamp(None=Multiply(A=@n20.Result) {constB=2})
  )
 ]
}
