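# Additive, unlit spline-mesh particle material for a comet or shooting star with a smoky trail.
# Comet texture alpha masks panning smoke noise, with dynamic parameters driving erosion, pan, width and randomness.
# World position offset thickens the spline via SplineThicken and pulls one end by EndWPO over particle lifetime.
# Time-of-day functions tint the HDR-boosted emissive and scale opacity, faintest by day and strongest at night.
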Material /Game/Effects/Fort_Effects/Materials/Comet/M_Comet_Shooting_Star_Spline
graph {
 properties {
  bUsedWithMeshParticles=true
  blendMode=BLEND_Additive
  pixelDepthOffsetMode=PDOM_Legacy
  previewMesh=/Game/Effects/Fort_Effects/Mesh/Splines/SM_SplineMesh_X_2UVs
  shadingModel=MSM_Unlit
  twoSided=true
  usedShadingModels=Unlit
 }
 nodes [
  n1 = DynamicParameter() {defaultValue=<rgba 1 1 1 0>, paramNames [SplineWidthMult, Erode, Pan, Random]}
  n2 = Multiply(
   A=ScalarParameter() {defaultValue=320, parameterName=SplineWidth}
   B=Multiply(A=@n1.SplineWidthMult, B=LinearInterpolate(Alpha=@n1.Random) {constA=0.65, constB=1.25})
  )
  n3 = ScalarParameter() {defaultValue=8, parameterName=SplineWidthTipMult}
  n4 = TextureCoordinate()
  n5 = Multiply(A=LinearInterpolate(Alpha=@n1.Random) {constA=0.4}, B=ParticleRelativeTime())
  n6 = Power(Base=ComponentMask(None=TextureCoordinate()) {g=true}) {constExponent=4}
  Add(
   A=MaterialFunctionCall(WidthBase=@n2, WidthTip=Multiply(A=@n2, B=@n3), UVs for Thickness=Power(Base=@n4) {constExponent=0.65}) {materialFunction=&SplineThicken}.WorldPositionOffset
   B=Multiply(A=Multiply(A=@n5, B=@n6), B=VectorParameter() {parameterName=EndWPO, defaultValue=<rgba 0 0 -30000 1>}.RGB)
  )
  Constant() {r=1}
  MaterialFunctionCall() {materialFunction=&DebugTimeSine}
  n7 = VectorParameter() {parameterName=EveningTint, defaultValue=<rgba 0.6791 0.8088 1 0>}
  n8 = VectorParameter() {parameterName=DayTint, defaultValue=<rgba 0.4479 1 0.8277 0>}
  n9 = VectorParameter() {parameterName=NightTint, defaultValue=<rgba 0.8 0.8 0.8 0>}
  n10 = ParticleColor()
  n11 = TextureSample() {texture=&/T_Comet_2D}
  n12 = TextureCoordinate()
  n13 = Multiply(
   A=VectorParameter() {parameterName=EdgeColor, defaultValue=<rgba 0.0365 0.7142 1 0>}.RGB
   B=Multiply(A=ComponentMask(None=@n12) {g=true}, B=OneMinus(None=@n11.A))
  )
  n14 = TextureCoordinate() {uTiling=0.7}
  n15 = Constant() {r=1}
  n16 = ScalarParameter() {defaultValue=1.6, parameterName=LengthTile}
  n17 = AppendVector(A=@n15, B=@n16)
  n18 = Add(A=Time(), B=@n1.Pan)
  n19 = TextureSample(UVs=Panner(Coordinate=Multiply(A=@n14, B=@n17), Time=@n18) {speedX=0.03, speedY=0.25}) {samplerType=SAMPLERTYPE_LinearColor, texture=/Game/Global/Textures/Default/Tileable_Patterns/T_Tile_Smoke_Small}
  n20 = Multiply(A=Multiply(A=@n19.R, B=@n19.R)) {constA=1, constB=4}
  n21 = AppendVector(A=@n16, B=@n15)
  n22 = Add(A=Multiply(A=TextureCoordinate() {coordinateIndex=1, uTiling=0.4, vTiling=0.85}, B=@n21)) {constB=0.075}
  n23 = Add(A=Multiply(A=@n21, B=TextureCoordinate() {coordinateIndex=1, uTiling=0.9, vTiling=0.6})) {constB=0.2}
  n24 = Multiply(
   A=TextureSample(UVs=Panner(Coordinate=@n22, Time=@n18) {speedX=-0.12}) {samplerType=SAMPLERTYPE_Grayscale, texture=&T_Smoke_Trail_Soft}.R
   B=TextureSample(UVs=Panner(Coordinate=@n23, Time=@n18) {speedX=-0.12}) {samplerType=SAMPLERTYPE_Grayscale, texture=&T_Smoke_Trail_Soft}.R
  )
  n25 = Multiply(A=@n11.A, B=Power(Base=Multiply(A=@n20, B=@n24) {constA=2, constB=4}, Exp=LinearInterpolate(Alpha=@n1.Erode) {constA=5, constB=0.6}))
  n26 = LinearInterpolate(
   A=VectorParameter() {parameterName=TailColor, defaultValue=<rgba 0.8787 1.409 3 0>}.RGB
   B=VectorParameter() {parameterName=HotColor, defaultValue=<rgba 3 0.914 0.375 0>}.RGB
   Alpha=Clamp(None=Multiply(A=@n25, B=@n1.Erode) {constB=0.5})
  )
  n27 = Multiply(A=@n10.RGB, B=Add(A=Multiply(A=@n11.RGB) {constB=2}, B=Add(A=@n13, B=@n26)))
  n28 = Clamp(None=Multiply(A=@n27) {constB=0.3})
  n29 = Power(Base=ComponentMask(None=TextureCoordinate()) {g=true}) {constExponent=5}
  n30 = LinearInterpolate(B=ScalarParameter() {defaultValue=2, parameterName=HDRIntensity}, Alpha=Clamp(None=Multiply(A=@n29) {constB=1.4})) {constA=0.4, constB=0.5}
  MaterialFunctionCall(Emissive=Multiply(
    A=MaterialFunctionCall(Color_Morning=@n7.RGB, Color_Day=@n8.RGB, Color_Evening=@n7.RGB, Color_Night=@n9.RGB) {materialFunction=&Color_Time_of_Day}.Result
    B=MaterialFunctionCall(LDR Input=@n28, HDR Tint=@n27, HDR Intensity Multiplier=@n30) {materialFunction=&DeriveHDRfromLDR}.RemappedRange
   )) {materialFunction=&MF_FX_EmissiveAdjustments}
  n31 = Constant() {r=0.4}
  Multiply(
   A=Multiply(A=@n10.A, B=@n25)
   B=MaterialFunctionCall(Color_Morning=@n31, Color_Day=Constant() {r=0.12}, Color_Evening=@n31, Color_Night=Constant() {r=1}) {materialFunction=&Color_Time_of_Day}.Result
  )
 ]
}
