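# Additive unlit particle material, for sprites including Niagara, that draws a glowing, churning star trail for comet effects.
# Two differently tiled noise samples pan and sum, time-offset per particle by a random value.
# A soft radial mask times that noise, cubed, blends dark to light HDR colors and, with particle alpha, forms opacity.
# Particle color tints the noise-modulated result, which passes through HDR remapping and emissive adjustment functions.
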Material /Game/Effects/Fort_Effects/Materials/Comet/M_Star_TrailParticle
graph {
 properties {
  bUsedWithNiagaraSprites=true
  bUsedWithParticleSprites=true
  blendMode=BLEND_Additive
  pixelDepthOffsetMode=PDOM_Legacy
  shadingModel=MSM_Unlit
  usedShadingModels=Unlit
 }
 nodes [
  n1 = ParticleColor()
  n2 = TextureCoordinate() {uTiling=0.7, vTiling=0.7}
  n3 = ParticleRandom()
  n4 = Multiply(A=Add(A=Multiply(A=@n3) {constB=2}, B=Time()), B=ScalarParameter() {defaultValue=0.6, parameterName=PanSpeed_1})
  n5 = TextureCoordinate() {uTiling=1.5, vTiling=1.5}
  n6 = Add(
   A=TextureSample(UVs=Panner(Coordinate=@n2, Time=@n4) {speedX=0.12, speedY=0.25}) {texture=/Game/Global/Textures/Default/Noise/T_TileNoise_02}.R
   B=TextureSample(UVs=Panner(Coordinate=@n5, Time=@n4) {speedX=-0.15, speedY=0.3}) {texture=&T_TileNoise_02}.R
  )
  n7 = VectorParameter() {parameterName=DarkColor, defaultValue=<rgba 1.4575 1.4062 3 0>}
  n8 = VectorParameter() {parameterName=LightColor, defaultValue=<rgba 21.5394 33.6055 50 0>}
  n9 = TextureCoordinate()
  n10 = Constant() {r=0.5}
  n11 = Power(Base=Multiply(A=SphereMask(A=@n9, B=@n10, Radius=@n10) {attenuationRadius=0.5, hardnessPercent=25}, B=@n6)) {constExponent=3}
  n12 = Multiply(A=@n1.RGB, B=Multiply(A=@n6, B=LinearInterpolate(A=@n7.RGB, B=@n8.RGB, Alpha=@n11)))
  MaterialFunctionCall(Emissive=MaterialFunctionCall(LDR Input=Clamp(None=@n12), HDR Tint=@n12, HDR Intensity Multiplier=ScalarParameter() {defaultValue=15, parameterName=HDRMult}) {materialFunction=&DeriveHDRfromLDR}.RemappedRange) {materialFunction=&MF_FX_EmissiveAdjustments}
  Multiply(A=Multiply(A=@n1.A, B=@n11)) {constB=2}
 ]
}
